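; Guard elevator for Heist bases that spawns AI reinforcements during a lockdown and opens its doors to release them.
; After a tunable delay, the server enables patrol point providers, capped by hotfix values keyed to playlist team size, and requests Living World spawning.
; The first spawn callback starts a timeline door opening with rep-notify material, collision and nav link updates.
; Once open, guards receive an enemy stim at a set location.
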(event EventBeginPlay
  (bind _returnvalue (Class|FortAthenaLivingWorldManager|GetLivingWorldManager))
  ;[0 1 0] Begin Play
  (|CallParentFunction)
  (CallFunction|SetDoorPositioninMaterials)
  (Networking|SwitchHasAuthority
    (:Authority
      (bind (_gamestate _playlist) (Class|FortAsyncActionCurrentPlaylistReady|PlaylistDataReadyAsync)
        (:OnReady
          (Utilities|FlowControl|Sequence
            (:then_0
              (Utilities|IsValid _playlist
                (:Is%20Valid
                  (bind _output_get (Variables|AI|SetCurrentPlaylistMaxTeamSize (Variables|/Script/FortniteGame.FortPlaylist|GetMaxTeamSize _playlist))) ; Cache this value to limit how many AI \nwe spawn when scalability is enabled.
                  (Fortnite|FortShippingLog self (Utilities|String|Append "CurrentPlaylistMaxTeamSize = " (Utilities|String|ToString(Integer) _output_get)))
                  ;[1 1 1] Setup Listeners for when an Ai Spawned
                  (Utilities|IsValid _returnvalue
                    (:Is%20Valid
                      (Variables|AI|SetLivingWorldManager _returnvalue)
                      (for _array_element (Variables|AI|GetPatrolPointProviders)
                        (Utilities|IsValid _array_element
                          (:Is%20Valid
                            (EventDispatchers|BindEventtoOnActorSpawnedResult _array_element (EventDispatchers|CreateEvent :!event Custom|OnReinforcementsSpawned))))))))))
            (:then_1
              (Fortnite|AI|SetComponentAffectNavigationGeneration (Variables|Default|GetNavLink)))))))))

(event Custom|OnReinforcementsSpawned
  (params (SpawnedActor :type Actor%20Object%20Reference) (bResult :type Boolean))
  ;[1 1 1] When AI Has Spawned
  (Utilities|FlowControl|Sequence
    (:then_0
      (Utilities|FlowControl|DoOnce false
        (:Completed
          (CallFunction|OpenDoor) ; Open the Doors
        )))
    (:then_1
      (if bResult
        (Utilities|Array|AddUnique (Variables|AI|GetSpawnedReinforcements) SpawnedActor)
        (if (Variables|Door|GetOpenDoorComplete)
          (CallFunction|SendStimtoSingleReinforcement SpawnedActor))))))

(event Custom|OnLockdownBegin
  ;[0.2552 0.1244 0.03277] OnLockdownBegin
  ; Open the doors and spawn the guards
  (Networking|SwitchHasAuthority
    (:Authority
      (Utilities|FlowControl|Delay (Variables|Lockdown|GetOnLockdownActivationDelay)) ; Instance editable so we can \nmanually tune staggered AI spawning
      (CallFunction|EnableAIPatrolPointProviders))))

(event Custom|OpenDoor
  (bind _dooropenedoffset (Variables|Door|GetDoorOpenedOffset))
  ;[0.04785 0.1004 0.3333] Open the elevator door this will happen when we verify that ai are in spawned in the elevator.
  (Utilities|FlowControl|DoOnce false
    (:Completed
      (Variables|Door|SetOpenDoorStart true)
      (Components|Timeline|SetPlayRate (Variables|BGA_HeistBases_GuardElevator|GetElevator_OpenDoors) (Math|Float|SafeDivide 1.0 (Variables|Door|GetElevatorOpeningTime))) ; set the play rate so it will scale \nthe normalized timeline to the \ndesired time
      (CallFunction|ActivateElevatorCosmetics)
      ;;[0 0.8379 1] Use timeline to animate door open. Server for door collision and replicated for visuals
      (bind (_direction _normalizedopencurve) (Timeline|Elevator_OpenDoors Play)
        (:Update
          (Transformation|SetRelativeLocation (Variables|Default|GetGuard_Door_Left) (Math|Vector|MakeVector 0.0 (* _normalizedopencurve _dooropenedoffset)) false false)
          (Transformation|SetRelativeLocation (Variables|Default|GetGuard_Door_Right) (Math|Vector|MakeVector 0.0 (* (* _normalizedopencurve _dooropenedoffset) -1.0)) false false))
        (:Finished
          (CallFunction|DeactivateElevatorCosmetics)
          (Variables|Door|SetOpenDoorComplete true)
          (CallFunction|SendStimtoAllReinforcements))
        (:Disable%20Collision
          (Variables|Door|SetOpenDoorCollisionDisable true)))
      (Networking|FlushNetDormancy))))

(event Custom|ActivateElevatorCosmetics
  ;[1 0 1] Cosmetics
  (Ability|ExecuteGameplayCueWithParamsReplicated self (Variables|Default|GetGC_ElevatorOpen) (Ability|GameplayCue|MakeGameplayCueParameters 0.0))
  (Networking|FlushNetDormancy))

(event Custom|DeactivateElevatorCosmetics)

(fn EnableAIPatrolPointProviders
  (Utilities|FlowControl|Sequence
    (:then_0
      ;[0.1407 0.1494 1] Set our limits on how many to enable
      (LocalVariables|SetLMaxProvidersToEnable (CallFunction|GetMaxAIPerElevator)))
    (:then_1
      ;[0.1407 0.1494 1] Iterate through our providers and enable up to our limit
      (bind _array_element (Utilities|Array|ForEachLoopwithBreak (Variables|AI|GetPatrolPointProviders))
        (:LoopBody
          (if (< (LocalVariables|GetLCurrentProvidersEnabled) (LocalVariables|GetLMaxProvidersToEnable))
            _
            (else
              _))
          (Utilities|IsValid _array_element
            (:Is%20Valid
              (Class|FortAthenaLivingWorldPointProviderInterface|EnablePointProvider_DEPRECATED _array_element)
              (Math|Integer|IncrementInt (LocalVariables|GetLCurrentProvidersEnabled)))))
        (:Completed
          _))
      (Fortnite|FortShippingLog self (Utilities|String|Append "Heist Reinforcements // " "Enabled " (Utilities|String|ToString(Integer) (LocalVariables|GetLCurrentProvidersEnabled)) " AI patrol providers"))
      (Class|FortAthenaLivingWorldManager|RequestEventGeneration (Variables|AI|GetLivingWorldManager)) ; Ask the Living World Manager to try to spawn Reinforcements now instead of waiting.
    )))

(fn GetMaxAIPerElevator
  (bind _hotfix_maxaiperelevator (Variables|Hotfix|GetHotfix_MaxAIPerElevator))
  (bind _currentplaylistmaxteamsize (Variables|AI|GetCurrentPlaylistMaxTeamSize))
  (if (ScalableFloat|GetValueasBool (Variables|Hotfix|GetHotfix_AIScalabilityTuningEnabled))
    (if (Utilities|Array|IsValidIndex _hotfix_maxaiperelevator _currentplaylistmaxteamsize)
      (return (ScalableFloat|GetValueasInteger (Utilities|Array|Get(acopy) _hotfix_maxaiperelevator _currentplaylistmaxteamsize)))
      (else
        _))
    (else
      _))
  (return (ScalableFloat|GetValueasInteger (Variables|Hotfix|GetHotfix_DefaultMaxAIPerElevator))))

(fn OnLockdownInitiated
  (CallFunction|OnLockdownBegin)
  (return true))

(fn OnRep_OpenDoorCollisionDisable
  (Collision|SetCollisionEnabled (Variables|Default|GetGuard_Door_Left) "NoCollision")
  (Networking|SwitchHasAuthority
    (:Authority
      (Fortnite|AI|SetComponentAffectNavigationGeneration (Variables|Default|GetNavLink) true))))

(fn OnRep_OpenDoorComplete
  (Rendering|SetVisibility (Variables|Default|GetGuard_Door_Left) false false))

(fn OnRep_OpenDoorStart
  (Rendering|Material|SetMaterial (Variables|Default|GetGuard_Door_Left) 0 "/Game/Environments/Asteria/Sets/EclipseVault/Materials/MI_EclipseVault_Elevator_Door.MI_EclipseVault_Elevator_Door"))

(fn SendStimToAllReinforcements
  (for _array_element (Variables|AI|GetSpawnedReinforcements)
    (CallFunction|SendStimtoSingleReinforcement _array_element)))

(fn SendStimToSingleReinforcement
  (params (AIActor :type Actor%20Object%20Reference))
  (bind _stimlocation (Variables|AI|GetStimLocation))
  (bind _asfort_player_pawn (Utilities|Casting|CastToFortPlayerPawn AIActor))
  (bind _asfort_athena_aibot_controller (Utilities|Casting|CastToFortAthenaAIBotController (Pawn|GetController _asfort_player_pawn)))
  (Utilities|IsValid _stimlocation
    (:Is%20Valid
      (Class|FortAthenaAIBotController|QueueStim _asfort_athena_aibot_controller _stimlocation (Transformation|GetActorLocation _stimlocation) (Math|Vector|MakeVector 1.0) "Enemy" 1.0 "None"))))

(fn SetDoorPositionInMaterials
  (Rendering|Material|SetDefaultCustomPrimitiveDataVector3 (Variables|Default|GetGuard_Door_Left) 0 (Transformation|GetActorLocation self)))

(fn ConstructionScript)
