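; Reinforcement point actor for Heist Bases guard elevators, exposing a spawn location and a hook for lockdown.
; OnLockdownInitiated calls the OnLockdownBegin custom event, which has no logic in this graph, and returns true.
; The construction script shifts the SpawnLocation component by SpawnLocationOffset as a world-space offset, teleporting without sweep.
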(event Custom|OnLockdownBegin)

(fn OnLockdownInitiated
  (CallFunction|OnLockdownBegin)
  (return true))

(fn ConstructionScript
  (Transformation|AddWorldOffset (Variables|Default|GetSpawnLocation) (Variables|Default|GetSpawnLocationOffset) false true))
